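# Translucent unlit Niagara sprite material for the PillowMill pickaxe effects, named as a temporary cutout.
# A texture parameter, defaulting to a slash background texture, sends its RGB through the shared MF_FX_EmissiveAdjustments function.
# The texture's alpha is multiplied by particle color alpha, so the particle system can fade the shape.
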Material /BRCosmetics/Weapons/FORT_Melee/Pickaxe_PillowMill/FX/Material/M_Pickaxe_PillowMill_Cutout_TEMP
graph {
 properties {
  bAutomaticallySetUsageInEditor=false
  bUsedWithNiagaraSprites=true
  blendMode=BLEND_Translucent
  shadingModel=MSM_Unlit
  usedShadingModels=Unlit
 }
 nodes [
  n1 = TextureSampleParameter2D() {parameterName=TextureInput, texture=/BRCosmetics/Weapons/FORT_Melee/Pickaxe_PillowMill/FX/Texture/T_PillowMill_Slash_Background}
  MaterialFunctionCall(Emissive=@n1.RGB) {materialFunction=&MF_FX_EmissiveAdjustments}
  Multiply(A=@n1.A, B=ParticleColor().A)
 ]
}
