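# Translucent unlit Niagara material for the toon-style swing trail of the PillowMill pickaxe.
# Step thresholds from DynamicParameter cut the streak texture, modulated by two counter-panning motion-line samples, into hard outer, middle and inner color bands, with a pink gradient in the middle.
# Opacity comes from the streak's red channel, eroded by a dissolve texture that particle alpha drives in quantized steps.
# A CameraOffset function supplies a depth-shifting position offset.
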Material /BRCosmetics/Weapons/FORT_Melee/Pickaxe_PillowMill/FX/Material/M_Pickaxe_PillowMill_Trail_Outline_Toon
graph {
 properties {
  bUsedWithNiagaraMeshParticles=true
  bUsedWithNiagaraRibbons=true
  bUsedWithNiagaraSprites=true
  blendMode=BLEND_Translucent
  pixelDepthOffsetMode=PDOM_Legacy
  shadingModel=MSM_Unlit
  usedShadingModels=Unlit
 }
 nodes [
  n1 = DynamicParameter() {defaultValue=<rgba 0.8 0.48 0.77 1.3>, paramNames [OuterColorCoverage, InnerColorCoverage, MiddleColorCoverage, ColorGradiantPan]}
  n2 = OneMinus(None=@n1.OuterColorCoverage)
  n3 = TextureSampleParameter2D() {parameterName=TextureInput, texture=/BRCosmetics/Weapons/FORT_Melee/Pickaxe_PillowMill/FX/Texture/T_Pickaxe_PillowMill_Trail_Streak}
  n4 = Power(Base=@n3.G, Exp=ScalarParameter() {defaultValue=0.25, parameterName=BaseTexturePower}) {constExponent=1}
  n5 = MaterialFunctionCall(Float2=TextureCoordinate()) {materialFunction=&BreakOutFloat2Components}
  n6 = Multiply(A=@n5.G)
  n7 = TextureSampleParameter2D(UVs=Panner(Coordinate=AppendVector(A=@n6, B=@n5.R)) {speedX=0.3}) {
   group=Motion Lines
   parameterName=Motion Lines Texture
   samplerSource=SSM_Wrap_WorldGroupSettings
   texture=/Game/Effects/Fort_Effects/Textures/Outlander/T_InTheZone_LightRays_01
  }
  n8 = Multiply(A=@n5.G) {constB=2}
  n9 = TextureSampleParameter2D(UVs=Panner(Coordinate=AppendVector(A=@n8, B=@n5.R)) {speedX=-0.7}) {
   group=Motion Lines
   parameterName=Motion Lines Texture
   samplerSource=SSM_Wrap_WorldGroupSettings
   texture=&T_InTheZone_LightRays_01
  }
  n10 = Add(A=Power(Base=@n7.R) {constExponent=1}, B=Power(Base=@n9.R))
  n11 = Multiply(A=@n4, B=LinearInterpolate(Alpha=Multiply(A=@n10) {constB=0.2}) {constA=0.85})
  n12 = OneMinus(None=ComponentMask(None=TextureCoordinate()) {r=true})
  n13 = LinearInterpolate(
   A=LinearInterpolate(
    B=VectorParameter() {group=Trail Color, parameterName=Outer_Color, defaultValue=<rgba 1 0.03 0.0462 1>}.RGB
    Alpha=Step(Y=@n2, X=@n11) {constX=0.9}
   )
   B=LinearInterpolate(
    A=Constant3Vector() {constant=<rgba 0.3593 0.0115 0.0957 1>}
    B=Constant3Vector() {constant=<rgba 0.9911 0.2213 0.4853 1>}
    Alpha=SmoothStep(Max=@n1.ColorGradiantPan, Value=@n12) {constMax=1.3}
   )
   Alpha=Step(Y=OneMinus(None=@n1.MiddleColorCoverage), X=@n11) {constY=0.3}
  )
  n14 = VectorParameter() {group=Trail Color, parameterName=Inner_Color, defaultValue=<rgba 2 2 2 1>}
  n15 = Step(Y=OneMinus(None=@n1.InnerColorCoverage), X=@n4) {constY=0.52}
  MaterialFunctionCall(Emissive=Multiply(A=ParticleColor().RGB, B=LinearInterpolate(A=@n13, B=@n14.RGB, Alpha=@n15))) {materialFunction=&MF_FX_EmissiveAdjustments}
  MaterialFunctionCall(Offset Amount=ScalarParameter() {defaultValue=50, parameterName=CameraOffset}) {materialFunction=/Engine/Functions/Engine_MaterialFunctions02/WorldPositionOffset/CameraOffset}
  n16 = ScalarParameter() {group=Dissolve, parameterName=Preview Dissolve - THIS SHOULD BE 0 NORMALLY, sliderMax=1}
  n17 = LinearInterpolate(A=OneMinus(None=ParticleColor().A), B=@n16, Alpha=@n16)
  n18 = Floor(None=ScalarParameter() {defaultValue=4, group=Dissolve, parameterName=Dissolve Steps})
  n19 = Divide(A=Floor(None=Multiply(A=@n17, B=@n18)), B=@n18)
  n20 = TextureCoordinate()
  n21 = Multiply(A=ComponentMask(None=@n20) {r=true}, B=ScalarParameter() {defaultValue=0.4, group=Dissolve, parameterName=Dissolve_Scale_U}) {constB=0.4}
  n22 = ComponentMask(None=@n20) {g=true}
  n23 = ScalarParameter() {defaultValue=0.45, group=Dissolve, parameterName=Dissolve_Scale_V}
  n24 = TextureSampleParameter2D(UVs=AppendVector(A=Add(A=@n21) {constB=0.3}, B=Multiply(A=@n22, B=@n23) {constB=0.3})) {
   group=Dissolve
   parameterName=Tex_Dissolve
   samplerType=SAMPLERTYPE_LinearGrayscale
   texture=/BRCosmetics/Weapons/FORT_Melee/Pickaxe_FearlessFlightMenace/FX/Textures/T_FearlessFlight_Menace_Trail_Dissolve_Frac
  }
  n25 = TextureCoordinate()
  n26 = Sine(None=Multiply(A=ComponentMask(None=@n25) {r=true}) {constB=0.5})
  n27 = Multiply(A=@n24.RGB, B=Saturate(None=Power(Base=@n26) {constExponent=1.3}))
  n28 = Multiply(A=@n27) {constB=0.1}
  n29 = Saturate(None=Multiply(A=ScalarParameter() {defaultValue=1.5, group=Trail Alpha Values, parameterName=Trail_Alpha_Intensity}, B=@n3.R) {constB=1.2})
  Multiply(A=Step(Y=Saturate(None=@n19), X=LinearInterpolate(A=@n28, B=@n27, Alpha=@n29)), B=@n29)
  ScalarParameter() {defaultValue=0.77, group=Trail Color, parameterName=Middle_Color_Coverage, sliderMax=1}
  ScalarParameter() {defaultValue=0.48, group=Trail Color, parameterName=Inner_Color_Coverage, sliderMax=1}
 ]
}
