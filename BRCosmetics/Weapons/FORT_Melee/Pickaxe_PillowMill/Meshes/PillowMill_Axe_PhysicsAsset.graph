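# Physics asset for the PillowMill Axe pickaxe cosmetic, a two-handed melee weapon mesh.
# It has a single body on the Melee_TwoHanded_Handle bone, shaped as one capsule offset and rotated relative to that bone.
# There are no constraints, so it defines one simple collision volume with no jointed or articulated parts.
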PhysicsAsset /BRCosmetics/Weapons/FORT_Melee/Pickaxe_PillowMill/Meshes/PillowMill_Axe_PhysicsAsset
bodies {Melee_TwoHanded_Handle [{shapeType=Capsule, length=51.5469, radius=5.5794, center=<-0.2887 1.0925 33.2234>, rotation=<rot 3.2489 -90 0.0062>}]}
constraints []
